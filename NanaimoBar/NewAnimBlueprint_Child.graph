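# Child-style Animation Blueprint for the Fortnite average male player skeleton, with no specific gameplay purpose evident.
# Its update event only forwards DeltaTime to the parent implementation.
# Loose nodes fetch the pawn owner and the FortPlayerPawn's skeletal mesh asset, but none connect to the update chain.
# An In Class variable defaults to NewAnimBlueprint_Linked2, though this event graph never reads it.
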AnimBlueprint /NanaimoBar/NewAnimBlueprint_Child
target_skeleton=&Fortnite_M_Avg_Player_Skeleton
graphs [EventGraph {nodes [
   n1 = Event BlueprintUpdateAnimation()
   > Parent: BlueprintUpdateAnimation!(execute, DeltaTimeX:float=@n1.DeltaTimeX) -> (then) {is=title K2Node_CallParentFunction}
   TryGetPawnOwner()
   n2 = Get FortPlayerPawn() -> (FortPlayerPawn:Fort Player Pawn Object Reference)
   GetSkeletalMeshAsset!(self:Skeletal Mesh Component Object Reference=Get Mesh(self=@n2)) -> (ReturnValue:Skeletal Mesh Object Reference)
   Get Mesh(self=@n2)
  ]}]
variables [In Class {type=Class, default=&/NewAnimBlueprint_Linked2.NewAnimBlueprint_Linked2_C}]
